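; Blueprint for the Pumpkin RPG rocket launcher in Athena, adding pumpkin-themed visual effects to the weapon.
; The construction script calls its parent, then attaches the PumpkinLauncher EyeFire effect to the weapon mesh at the RocketFX socket, keeping relative transforms.
; The muzzle flash event calls the parent, then toggles the active state of the DrunkenQuad back muzzle effect component.
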(event EventMuzzleFlashFX
  (params (Persistent%20Fire :type Boolean))
  (|CallParentFunction false)
  (Components|Activation|ToggleActive (Variables|Default|GetNS_RocketLauncher_DrunkenQuad_Muzzle_Back)))

(fn ConstructionScript
  (|CallParentFunction)
  (Transformation|AttachComponentToComponent (Variables|Default|GetNS_PumpkinLauncher_EyeFire) (Variables|Visual|GetWeaponMesh) "RocketFX" "KeepRelative" "KeepRelative" "KeepRelative" true))
