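; Weapon Blueprint for the LumpyRug mythic lock-on rocket launcher, whose projectiles home on a locked player pawn.
; Aiming down sights sends press/release gameplay events, while a locally spawned widget actor tracks the potential target's spine on clients.
; On authority, projectile spawns latch the lock target as sticky and apply homing, clearing it after a retriggerable delay.
; Remote spawns activate the rear muzzle effect and cycle muzzle-flash offsets through six positions.
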(event EventTick
  (params (DeltaSeconds :type Float%20%28single-precision%29))
  (bind _potentialtarget (Variables|Default|GetPotentialTarget))
  ;[1 1 1] Actor Based Lock On Widget
  (if (not (Networking|IsServer))
    (Utilities|IsValid _potentialtarget
      (:Is%20Valid
        (bind _asfort_player_pawn (Utilities|Casting|CastToFortPlayerPawn _potentialtarget))
        (Transformation|SetActorLocation (Variables|Default|GetWidgetActor) (Transformation|GetSocketLocation (Variables|/Script/Engine.Character|GetMesh _asfort_player_pawn) "spine_04") false false)
        (Rendering|SetActorHiddenInGame (Variables|Default|GetWidgetActor) false))
      (:Is%20Not%20Valid
        (Rendering|SetActorHiddenInGame (Variables|Default|GetWidgetActor) true)))))

(event FortWeapon|EventOnEquippedAndAttached
  (Default|BindEventtoOnAimDownSightsChanged (delegate Custom|OnAimPassLocation))
  (Utilities|FlowControl|Sequence
    (:then_1
      (if (Pawn|IsLocallyControlled (Variables|Actor|GetInstigator))
        (bind _returnvalue (Game|SpawnActorfromClass "/Game/Athena/Items/Mythics/LumpyRug/Athena_LockOnWidgetActor_Proto_LumpyRug.Athena_LockOnWidgetActor_Proto_LumpyRug_C" (Math|Conversions|ToTransform(Vector) (Transformation|GetActorLocation)) "Undefined" "OverrideRootScale" 0 0))
        (Variables|Default|SetWidgetActor _returnvalue)
        (bind _asathena_lumpy_rug_lock_on_widget (Utilities|Casting|CastToAthena_LumpyRug_LockOnWidget (UserInterface|GetUserWidgetObject (Variables|/Game/Athena/Items/Mythics/LumpyRug/Athena_LockOnWidgetActor_Proto_LumpyRug.Athena_LockOnWidgetActor_Proto_LumpyRug_C|GetWidget _returnvalue))))
        (Variables|Default|SetLockOnWidget _asathena_lumpy_rug_lock_on_widget)
        (Rendering|SetActorHiddenInGame (Variables|Default|GetWidgetActor) true)))))

(event FortWeapon|EventOnLocalUnEquip
  (Actor|DestroyActor (stale FortWeapon|EventOnEquippedAndAttached _returnvalue)))

(event Custom|OnAimPassLocation
  (params (AimDownSights :type Boolean))
  (bind _returnvalue (Game|GetInstigator))
  (if AimDownSights
    (Ability|SendGameplayEventtoActor _returnvalue "(TagName=\"Event.Athena.Item.LumpyRug.OnPress\")")
    (else
      (Ability|SendGameplayEventtoActor _returnvalue "(TagName=\"Event.Athena.Item.LumpyRug.OnRelease\")"))))

(event FortWeapon|EventOnSpawnProjectile
  (params (SpawnProjectile :type Fort%20Projectile%20Base%20Object%20Reference))
  (Networking|SwitchHasAuthority
    (:Authority
      (Utilities|FlowControl|Sequence
        (:then_0
          (Utilities|IsValid (Variables|Default|GetLockOnTarget)
            (:Is%20Valid
              _)))
        (:then_1
          (Utilities|IsValid (Variables|Default|GetLockOnTargetSticky)
            (:Is%20Valid
              (bind _asfort_player_pawn (Utilities|Casting|CastToFortPlayerPawn (Variables|Default|GetLockOnTargetSticky)))
              (Projectile|SetHomingData SpawnProjectile SpawnProjectile (Utilities|Struct|MakeProjectileHomingData "LockOn" 0.0 (Variables|Default|GetTurnSpeedMax) (Variables|Default|GetTimeUntilMax) (Variables|Default|GetHomingDistanceThreshold) (Math|Random|RandomIntegerinRange 0 14) _asfort_player_pawn) :!class /Script/FortniteGame.FortProjectileMovementInterface))))
        (:then_2
          (Utilities|FlowControl|RetriggerableDelay 1.0)
          (Variables|Default|SetLockOnTargetSticky 0)
          (goto _label Reset)))
      (Utilities|FlowControl|DoOnce false :!label _label
        (:Completed
          (Variables|Default|SetLockOnTargetSticky (Variables|Default|GetLockOnTarget)))))
    (:Remote
      (Components|Activation|Activate (Variables|Default|GetMuzzleBack) true)
      (bind _counter (Utilities|FlowControl|DoN 6 :!label _label_2)
        (:Exit
          (Variables|Default|SetMuzzle%20Relative%20Location (Utilities|Select :Index _counter :Option%200 "0, 0, 0" :Option%201 "0, 0,8.000000" :Option%202 "0.000000,-5.000000,6.000000" :Option%203 "0.000000,-5.000000,0.200000" :Option%204 "0.000000,0.000000,-2.500000" :Option%205 "0,4.500000,0.200000" :Option%206 "0,4.500000,6.000000"))
          (if (>= _counter 6)
            (goto _label_2 Reset)))))))

(event EventMuzzleFlashFX
  (params (Persistent%20Fire :type Boolean))
  (bind _spawned_muzzle_system (Variables|Effects|Muzzle|GetSpawned%20Muzzle%20System))
  (|CallParentFunction Persistent%20Fire)
  (Transformation|SetRelativeLocation _spawned_muzzle_system (Variables|Default|GetMuzzle%20Relative%20Location) false false)
  (Transformation|SetRelativeScale3D _spawned_muzzle_system "1.000000,1.000000,1.000000"))

(fn ConstructionScript
  (|CallParentFunction))
