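; Projectile fired by the Ferret biplane's dual guns in Athena, built on the rocket projectile setup.
; On spawn it starts a fuse timer, the server ignores collision with the firing vehicle and passengers, and remote clients feed muzzle origin to mesh materials.
; On stop or fuse expiry, biplanes or their drivers can take scalable-float-gated direct hits instead of explosions.
; Cosmetic handlers play water-aware effects, sound, camera shake and force feedback.
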(event EventBeginPlay
  ;[1 1 1] EVENT -> Begin Play
  (Utilities|Time|SetTimerbyFunctionName self "FuseTimerMax" 25.0)
  (Collision|SetIgnoreInstigatorCollision true)
  (bind _asfort_player_pawn (Utilities|Casting|CastToFortPlayerPawn (Variables|Actor|GetInstigator)))
  (bind _output_get (Variables|ProjectileRocket|SetPlayerPawn _asfort_player_pawn))
  (Networking|SwitchHasAuthority
    (:Authority
      (bind _returnvalue (Vehicle|GetVehicleActor (Vehicle|GetVehicleInterface _output_get) :!class /Script/FortniteGame.FortVehicleInterface))
      (Collision|IgnoreActorwhenMoving (Variables|Collision|GetCapsuleComponent) _returnvalue true)
      (for _array_element (Vehicle|GetAllPassengers (Vehicle|GetVehicleInterface _output_get))
        (Collision|IgnoreActorwhenMoving (Variables|Collision|GetCapsuleComponent) _array_element true)))
    (:Remote
      (bind _asfort_athena_ferret_vehicle (Utilities|Casting|CastToFortAthenaFerretVehicle (Utilities|ToObject(Interface) (Vehicle|GetVehicleInterface _output_get)))
        (:then
          (if (and (Variables|/Script/FortniteGame.FortFerretVehicleConfigs|GetbUseBiplaneComponentsForProjectileOrigins (Variables|/Script/FortniteGame.FortAthenaFerretVehicle|GetFortFerretVehicleConfigs _asfort_athena_ferret_vehicle)) (Replay|GetIsPlayingReplay))
            (bind _asb_ranged_dual_generic_ferret_vehicle (Utilities|Casting|CastToB_Ranged_Dual_Generic_FerretVehicle (Variables|/Script/FortniteGame.FortPawn|GetCurrentWeapon _output_get)))
            (CallFunction|SetMaterialParameters (Transformation|GetWorldLocation (Variables|/Script/FortniteGame.FortAthenaFerretVehicle|GetRightGunMuzzle _asfort_athena_ferret_vehicle)) (Transformation|GetWorldLocation (Variables|/Script/FortniteGame.FortAthenaFerretVehicle|GetLeftGunMuzzle _asfort_athena_ferret_vehicle)) (RangedWeaponDual|GetLastFireHand _asb_ranged_dual_generic_ferret_vehicle))
            (else
              _)))
        (:CastFailed
          _))
      (bind _asb_ranged_dual_generic_ferret_vehicle_2 (Utilities|Casting|CastToB_Ranged_Dual_Generic_FerretVehicle (Variables|/Script/FortniteGame.FortPawn|GetCurrentWeapon _output_get)))
      (CallFunction|SetMaterialParameters (Transformation|GetSocketLocation (Class|FortWeaponRangedDual|GetLeftHandWeaponMesh _asb_ranged_dual_generic_ferret_vehicle_2) (Effects|GetMuzzleSocketName _asb_ranged_dual_generic_ferret_vehicle_2 0)) (Transformation|GetSocketLocation (Variables|/Script/FortniteGame.FortWeapon|GetWeaponMesh _asb_ranged_dual_generic_ferret_vehicle_2) (Effects|GetMuzzleSocketName _asb_ranged_dual_generic_ferret_vehicle_2 0)) (RangedWeaponDual|GetLastFireHand _asb_ranged_dual_generic_ferret_vehicle_2)))))

(event EventOnStop
  (params (Hit :type Hit%20Result%20Structure%20%28by%20ref%29))
  (bind (_bblockinghit _binitialoverlap _time _distance _location _impactpoint _normal _impactnormal _physmat _hitactor) (Collision|BreakHitResult Hit))
  (bind (_asfort_athena_ferret_vehicle _bsuccess) (Utilities|Casting|CastToFortAthenaFerretVehicle _hitactor))
  ;[1 1 1] Impact and Explosion Stuff
  ;;[1 1 1] EVENT-> On Stop
  (Variables|Effects|SetStopLocation (+ _impactpoint (* _impactnormal (Variables|Effects|GetExplosion%20Impact%20Offset))))
  (Variables|Effects|SetStopRotZ (Math|Rotator|MakeRotfromX _impactnormal))
  (Variables|Default|SetHitWaterBodyActor (Utilities|Casting|CastToFortWaterBodyActor _hitactor))
  (if (Delivery|IsBeingKilled)
    _
    (elif (> (ScalableFloat|GetValueatLevel (Variables|ProjectileRocket|GetDoDirectDamageToBiplanes?)) 0.0)
      ;;[1 1 1] If we hit another Biplane or the driver, do direct damage to the Biplane instead of explosive
      (bind _asfort_player_pawn (Utilities|Casting|CastToFortPlayerPawn _hitactor)
        (:then
          (if (and (.bSuccess (Utilities|Casting|CastToFortAthenaFerretVehicle (Utilities|ToObject(Interface) (Vehicle|GetVehicleInterface _asfort_player_pawn)))) (Vehicle|IsDrivingVehicle _asfort_player_pawn))
            (Delivery|DoSingleActorHit (Collision|MakeHitResult false false 0.0 0.0 (Class|FortVehicleInterface|GetOwnerLocation (Vehicle|GetVehicleInterface _asfort_player_pawn)) (Class|FortVehicleInterface|GetOwnerLocation (Vehicle|GetVehicleInterface _asfort_player_pawn)) "0,0,1" "0,0,1" 0 (Utilities|Casting|CastToFortAthenaFerretVehicle (Utilities|ToObject(Interface) (Vehicle|GetVehicleInterface _asfort_player_pawn)))))
            (CallFunction|CustomExploded :!label _label)
            (else
              _)))
        (:CastFailed
          (if _bsuccess
            (Delivery|DoSingleActorHit (Collision|MakeHitResult false false 0.0 0.0 (Transformation|GetActorLocation _hitactor) (Transformation|GetActorLocation _hitactor) "0,0,1" "0,0,1" 0 _hitactor))
            (goto _label)
            (else
              _))))))
  (Delivery|DoExplosion) ; Do regular explosion if it didn't hit a Biplane driver or another Biplane
)

(event Custom|FuseTimerMax
  ;[1 1 1] Impact and Explosion Stuff
  ;;[1 1 1] EVENT-> On Stop
  (Variables|Effects|SetTimerMaxReached? true)
  (Variables|Effects|SetStopLocation (Transformation|GetWorldLocation (Variables|Projectile_Rocket|GetMesh)))
  (if (Delivery|IsBeingKilled)
    _
    (elif (> (ScalableFloat|GetValueatLevel (Variables|ProjectileRocket|GetDoDirectDamageToBiplanes?)) 0.0)
      ;;[1 1 1] If we hit another Biplane or the driver, do direct damage to the Biplane instead of explosive
      (bind _asfort_player_pawn (Utilities|Casting|CastToFortPlayerPawn (.HitActor (Collision|BreakHitResult (stale EventOnStop Hit))))
        (:then
          (if (and (.bSuccess (Utilities|Casting|CastToFortAthenaFerretVehicle (Utilities|ToObject(Interface) (Vehicle|GetVehicleInterface _asfort_player_pawn)))) (Vehicle|IsDrivingVehicle _asfort_player_pawn))
            (Delivery|DoSingleActorHit (Collision|MakeHitResult false false 0.0 0.0 (Class|FortVehicleInterface|GetOwnerLocation (Vehicle|GetVehicleInterface _asfort_player_pawn)) (Class|FortVehicleInterface|GetOwnerLocation (Vehicle|GetVehicleInterface _asfort_player_pawn)) "0,0,1" "0,0,1" 0 (Utilities|Casting|CastToFortAthenaFerretVehicle (Utilities|ToObject(Interface) (Vehicle|GetVehicleInterface _asfort_player_pawn)))))
            (CallFunction|CustomExploded :!label _label)
            (else
              _)))
        (:CastFailed
          (if (.bSuccess (Utilities|Casting|CastToFortAthenaFerretVehicle (.HitActor (Collision|BreakHitResult (stale EventOnStop Hit)))))
            (Delivery|DoSingleActorHit (Collision|MakeHitResult false false 0.0 0.0 (Transformation|GetActorLocation (.HitActor (Collision|BreakHitResult (stale EventOnStop Hit)))) (Transformation|GetActorLocation (.HitActor (Collision|BreakHitResult (stale EventOnStop Hit)))) "0,0,1" "0,0,1" 0 (.HitActor (Collision|BreakHitResult (stale EventOnStop Hit)))))
            (goto _label)
            (else
              _))))))
  (Delivery|DoExplosion) ; Do regular explosion if it didn't hit a Biplane driver or another Biplane
)

(event Custom|CustomExploded
  (bind _stoplocation (Variables|Effects|GetStopLocation))
  ;[1 1 1] Impact and Explosion Stuff
  ;;[1 1 1] EVENT -> On Exploded
  (Networking|CanExecuteCosmeticEvents
    (:True
      (Utilities|FlowControl|Sequence
        (:then_0
          ;;[0.03285 0.1156 0.1667] Spawn the necessary FXs based on if a water body actor was hit or not.
          (bind _hitwaterbodyactor (Variables|Default|GetHitWaterBodyActor)
            (:then
              (bind _outwatersurfaceinfo (WaterBody|GetWaterSurfaceInfoStruct _hitwaterbodyactor (Variables|Effects|GetStopLocation) true false false))
              (bind _returnvalue (Effects|Components|ParticleSystem|SpawnFXAtLocation (Variables|Effects|GetNS_HitWater) (Utilities|Struct|BreakWaterSurfaceInfo _outwatersurfaceinfo) (Math|Conversions|RotationFromXVector (.WaterPlaneNormal (Utilities|Struct|BreakWaterSurfaceInfo _outwatersurfaceinfo)))))
              (Utilities|IsValid _returnvalue
                (:Is%20Valid
                  (Effects|Components|ParticleSystem|SetFloatParameter _returnvalue "WaterDepth" (.WaterDepth (Utilities|Struct|BreakWaterSurfaceInfo _outwatersurfaceinfo)))
                  (Effects|Components|ParticleSystem|SetFloatParameter _returnvalue "WaterBodyIndex" (Math|Conversions|ToFloat(Integer) (.WaterBodyIdx (Utilities|Struct|BreakWaterSurfaceInfo _outwatersurfaceinfo)))))
                (:Is%20Not%20Valid
                  _))
              (Audio|PlaySoundatLocation (Variables|Sounds|GetProjectile_HitWater_Sound) (Variables|Effects|GetStopLocation)))
            (:else)) ; Hit water?
        )
        (:then_1
          ;;[0.06325 0.125 0.07605] Spawn the projectile explosion FX.
          (bind _returnvalue_2 (Effects|SpawnEmitteratLocation (select (Utilities|IsValid (Variables|Default|GetHitWaterBodyActor) :!class /Script/Engine.KismetSystemLibrary) (Variables|Effects|GetWater_Explosion_PS) (Variables|Effects|GetExplosion_Generic_PS)) (Variables|Effects|GetStopLocation) (Variables|Effects|GetStopRotZ)))
          (Utilities|IsValid _returnvalue_2
            (:Is%20Valid
              (Rendering|SetRenderCustomDepth _returnvalue_2 true))
            (:Is%20Not%20Valid
              _))
          (Audio|PlaySoundatLocation (Variables|Sounds|GetProjectile_Explosion_Sound) _stoplocation)
          ;;[0.1198 0.0564 0.1132] Shake camera.
          (Camera|PlayWorldCameraShake "/Game/Effects/Fort_Effects/BluePrints/CameraShakes/B_Small_Vertical_Jolt_CameraShake.B_Small_Vertical_Jolt_CameraShake_C" _stoplocation 5e+01 9e+02)
          (Fortnite|PlayLocalForceFeedbackatLocationMulti (Variables|ProjectileRocket|GetExplosionForceFeedbackNear) (Variables|ProjectileRocket|GetExplosionForceFeedbackFar) _stoplocation 1.4e+02 3e+02 "Rocket"))))))

(event EventOnExploded
  (params (HitActors :type Array%20of%20Actor%20Object%20References) (HitResults :type Array%20of%20Hit%20Result%20Structures))
  (bind _stoplocation (Variables|Effects|GetStopLocation))
  ;[1 1 1] Impact and Explosion Stuff
  ;;[1 1 1] EVENT -> On Exploded
  (Networking|CanExecuteCosmeticEvents
    (:True
      (Utilities|FlowControl|Sequence
        (:then_0
          ;;[0.03285 0.1156 0.1667] Spawn the necessary FXs based on if a water body actor was hit or not.
          (bind _hitwaterbodyactor (Variables|Default|GetHitWaterBodyActor)
            (:then
              (bind _outwatersurfaceinfo (WaterBody|GetWaterSurfaceInfoStruct _hitwaterbodyactor (Variables|Effects|GetStopLocation) true false false))
              (bind _returnvalue (Effects|Components|ParticleSystem|SpawnFXAtLocation (Variables|Effects|GetNS_HitWater) (Utilities|Struct|BreakWaterSurfaceInfo _outwatersurfaceinfo) (Math|Conversions|RotationFromXVector (.WaterPlaneNormal (Utilities|Struct|BreakWaterSurfaceInfo _outwatersurfaceinfo)))))
              (Utilities|IsValid _returnvalue
                (:Is%20Valid
                  (Effects|Components|ParticleSystem|SetFloatParameter _returnvalue "WaterDepth" (.WaterDepth (Utilities|Struct|BreakWaterSurfaceInfo _outwatersurfaceinfo)))
                  (Effects|Components|ParticleSystem|SetFloatParameter _returnvalue "WaterBodyIndex" (Math|Conversions|ToFloat(Integer) (.WaterBodyIdx (Utilities|Struct|BreakWaterSurfaceInfo _outwatersurfaceinfo)))))
                (:Is%20Not%20Valid
                  _))
              (Audio|PlaySoundatLocation (Variables|Sounds|GetProjectile_HitWater_Sound) (Variables|Effects|GetStopLocation)))
            (:else)) ; Hit water?
        )
        (:then_1
          ;;[0.06325 0.125 0.07605] Spawn the projectile explosion FX.
          (bind _returnvalue_2 (Effects|SpawnEmitteratLocation (select (Utilities|IsValid (Variables|Default|GetHitWaterBodyActor) :!class /Script/Engine.KismetSystemLibrary) (Variables|Effects|GetWater_Explosion_PS) (Variables|Effects|GetExplosion_Generic_PS)) (Variables|Effects|GetStopLocation) (Variables|Effects|GetStopRotZ)))
          (Utilities|IsValid _returnvalue_2
            (:Is%20Valid
              (Rendering|SetRenderCustomDepth _returnvalue_2 true))
            (:Is%20Not%20Valid
              _))
          (Audio|PlaySoundatLocation (Variables|Sounds|GetProjectile_Explosion_Sound) _stoplocation)
          ;;[0.1198 0.0564 0.1132] Shake camera.
          (Camera|PlayWorldCameraShake "/Game/Effects/Fort_Effects/BluePrints/CameraShakes/B_Small_Vertical_Jolt_CameraShake.B_Small_Vertical_Jolt_CameraShake_C" _stoplocation 5e+01 9e+02)
          (Fortnite|PlayLocalForceFeedbackatLocationMulti (Variables|ProjectileRocket|GetExplosionForceFeedbackNear) (Variables|ProjectileRocket|GetExplosionForceFeedbackFar) _stoplocation 1.4e+02 3e+02 "Rocket"))))))

(event EventDestroyed
  ;[1 1 1] Impact and Explosion Stuff
  ;;[1 1 1] EVENT -> Destroyed
  (Networking|CanExecuteCosmeticEvents
    (:True
      (Components|Activation|SetActive (Variables|Effects|GetRocketTrailPS) false false))))

(event EventOnTouched
  (params (OtherActor :type Actor%20Object%20Reference) (OtherComp :type Primitive%20Component%20Object%20Reference) (HitResult :type Hit%20Result%20Structure%20%28by%20ref%29) (bIsOverlap :type Boolean))
  (bind (_bblockinghit _binitialoverlap _time _distance _location _impactpoint _normal _impactnormal _physmat _hitactor _hitcomponent) (Collision|BreakHitResult HitResult))
  ;[1 1 1] EVENT -> On Touched - Dont explode if we are a snowman and get hit (Fix FORT-140529).
  (bind _asbuilding_gameplay_actor_balloon (Utilities|Casting|CastToBuildingGameplayActorBalloon OtherActor)
    (:then
      (if (Team|OnSameTeam (Actor|GetOwner _asbuilding_gameplay_actor_balloon) (Game|GetInstigator))
        _
        (else
          _)))
    (:CastFailed
      _))
  ;[1 1 1] Bounce off "NoBounce"
  (Utilities|IsValid _hitcomponent
    (:Is%20Valid
      (if (Components|ComponentHasTag _hitcomponent "NoStick")
        (Variables|/Script/Engine.MovementComponent|SetVelocity (* (Variables|/Script/Engine.MovementComponent|GetVelocity (Variables|Movement|GetProjectileMovementComponent)) 2.0) (Variables|Movement|GetProjectileMovementComponent))
        (else
          _)))
    (:Is%20Not%20Valid
      _))
  (Components|Activation|SetActive (Variables|Effects|GetRocketTrailPS) false false)
  (Rendering|SetVisibility (Variables|Projectile_Rocket|GetMesh) false false)
  (Variables|Projectile|SetbStopSimulatingOnHit true))

(fn SetMaterialParameters
  (params (LEFT :type Vector) (RIGHT :type Vector) (Index :type EDualWeaponHand%20Enum :default "LEFT"))
  (bind _mesh (Variables|Projectile_Rocket|GetMesh))
  (Rendering|Material|SetVectorParameterValueonMaterials _mesh "BulletOrigin" (+ (* (Transformation|GetActorForwardVector) 5e+01) (Utilities|Select :Index Index :LEFT LEFT :RIGHT RIGHT)))
  (Rendering|Material|SetScalarParameterValueonMaterials _mesh "StartTime" (Utilities|Time|GetGameTimeinSeconds)))

(fn ConstructionScript
  (Effects|Components|ParticleSystem|SetTemplate (Variables|Default|GetRibbon_Trail_PSC) (Variables|Effects|GetRibbon_Trail_PS))
  (Effects|Components|ParticleSystem|SetTemplate (Variables|Default|GetExplosion_Generic_PSC) (Variables|Effects|GetExplosion_Generic_PS))
  (Effects|Components|ParticleSystem|SetTemplate (Variables|Default|GetExplosion_Flesh_Damage_PSC) (Variables|Effects|GetExplosion_Flesh_Damage_PS)))
