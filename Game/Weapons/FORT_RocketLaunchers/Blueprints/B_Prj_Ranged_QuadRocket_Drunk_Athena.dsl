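; Quad rocket launcher projectile that homes on the owner's aim point, switching from LockOn to DrunkArtillery homing after a zero-length delay.
; Touching a NoStick component doubles velocity; other touches allow stop-on-hit, and stopping triggers the explosion.
; A fuse timer forces detonation, while an optional delay postpones water-touch stopping and detonates rockets already submerged.
; Cosmetic-only logic handles the trail, water or default explosion effects and sounds, and camera-shake cue.
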(event EventBeginPlay
  ;[1 1 1] Begin Play
  (bind _returnvalue (Fortnite|GetPlayerAimPoint (Fortnite|GetFortPlayerControllerfromActor (Actor|GetOwner)) 2.5e+04 (Variables|Default|GetIgnoredActors)))
  (Projectile|SetHomingData 0 0 (Utilities|Struct|MakeProjectileHomingData "LockOn" 1e+02 3e+02 1.0 0.25 -1 0 0 _returnvalue) :!class /Script/FortniteGame.FortProjectileMovementInterface)
  (Utilities|FlowControl|Delay 0.0)
  (Projectile|SetHomingData 0 0 (Utilities|Struct|MakeProjectileHomingData "DrunkArtillery" 1e+02 3e+02 1.0 0.25 -1 0 0 _returnvalue) :!class /Script/FortniteGame.FortProjectileMovementInterface)
  (Utilities|FlowControl|Sequence
    (:then_0
      ;[1 1 1] Timer to enable touching water to cause an explosion
      (if (> (Variables|Default|GetWater%20Interaction%20Delay) 0.0)
        (Utilities|Time|SetTimerbyEvent (EventDispatchers|CreateEvent :!event Custom|RocketWaterDelay) (Variables|Default|GetWater%20Interaction%20Delay))
        (else
          (Variables|Projectile|SetTouchWaterBehavior "StopIfStopSimulatingOnHit"))))
    (:then_1
      (Utilities|Time|SetTimerbyFunctionName self "FuseTimerMax" 15)
      (Networking|CanExecuteCosmeticEvents
        (:True
          (bind _returnvalue_2 (Effects|SpawnFXSystemAttached :FXSystemAsset "/Game/Weapons/FORT_RocketLaunchers/Effects/NS_RocketTrail_DrunkenQuad_01.NS_RocketTrail_DrunkenQuad_01" :AttachToComponent (Variables|Projectile_Rocket|GetMesh) :AttachPointName "None" :Rotation (Math|Rotator|MakeRotator (Math|Random|RandomFloatinRange 0.0 360)) :Scale "1.000000,1.000000,1.000000" :LocationType "KeepRelativeOffset" :bAutoDestroy true :PoolingMethod "None" :bAutoActivate false))
          (Rendering|SetRenderCustomDepth _returnvalue_2 true))))))

(event EventTick
  (params (DeltaSeconds :type Float%20%28single-precision%29)))

(event EventOnTouched
  (params (OtherActor :type Actor%20Object%20Reference) (OtherComp :type Primitive%20Component%20Object%20Reference) (HitResult :type Hit%20Result%20Structure%20%28by%20ref%29) (bIsOverlap :type Boolean))
  (bind (_bblockinghit _binitialoverlap _time _distance _location _impactpoint _normal _impactnormal _physmat _hitactor _hitcomponent) (Collision|BreakHitResult HitResult))
  ;[1 1 1] Bounce off "NoBounce"
  (if (Components|ComponentHasTag _hitcomponent "NoStick")
    (Variables|/Script/Engine.MovementComponent|SetVelocity (* (Variables|/Script/Engine.MovementComponent|GetVelocity (Variables|Movement|GetProjectileMovementComponent)) 2.0) (Variables|Movement|GetProjectileMovementComponent))
    (else
      (Variables|Projectile|SetbStopSimulatingOnHit true))))

(event Custom|RocketWaterDelay
  ;[1 1 1] Water Explosion Delay
  (if (and (Utilities|IsValid (Variables|/Script/Engine.MovementComponent|GetUpdatedComponent (Variables|Movement|GetProjectileMovementComponent)) :!class /Script/Engine.KismetSystemLibrary) (not (Delivery|IsBeingKilled)))
    (Variables|Projectile|SetTouchWaterBehavior "StopIfStopSimulatingOnHit")
    (if (Water|IsinWater (Variables|Water|GetWaterInteractionComponent))
      (CallFunction|FuseTimerMax))))

(event EventOnStop
  (params (Hit :type Hit%20Result%20Structure%20%28by%20ref%29))
  (bind (_bblockinghit _binitialoverlap _time _distance _location _impactpoint _normal _impactnormal _physmat _hitactor) (Collision|BreakHitResult Hit))
  (bind _surfacetype (Variables|/Script/PhysicsCore.PhysicalMaterial|GetSurfaceType _physmat))
  ;[1 1 1] Impact and Explosion Stuff
  (Variables|Effects|SetDecalLocation _impactpoint)
  (Variables|Effects|SetStopLocation (+ _impactpoint (* _impactnormal (Variables|Effects|GetExplosion%20Impact%20Offset))))
  (Variables|Effects|SetStopRotZ (Math|Rotator|MakeRotfromZ _impactnormal))
  (Variables|Default|SetSurfaceType _surfacetype)
  (Class|BPSurfaceTypeSoundComponent|SetSurfaceType (Variables|Default|GetBP_SurfaceTypeSoundComponent) _surfacetype _hitactor)
  (if (not (Delivery|IsBeingKilled))
    (Delivery|DoExplosion)
    (Fortnite|ShowSoundIndicatorLocal :Location (Transformation|GetWorldLocation (Variables|Projectile_Rocket|GetMesh)) :MaxAudibleDistance 5e+03 :IndicatorType "Gunshot" :Instigator self :Tint "(R=0.000000,G=0.000000,B=0.000000,A=0.000000)" :AttenuationMaxRangeMultiplier 1.0 :AttenuationIntensityMultiplier 1.0)))

(event Custom|FuseTimerMax
  (bind _mesh (Variables|Projectile_Rocket|GetMesh))
  (bind _returnvalue (Transformation|GetWorldLocation _mesh))
  ;[1 1 1] Impact and Explosion Stuff
  (Variables|Effects|SetTimerMaxReached? true)
  (Variables|Effects|SetStopLocation _returnvalue)
  (if (not (Delivery|IsBeingKilled))
    (Delivery|DoExplosion)
    (Fortnite|ShowSoundIndicatorLocal :Location _returnvalue :MaxAudibleDistance 5e+03 :IndicatorType "Gunshot" :Instigator self :Tint "(R=0.000000,G=0.000000,B=0.000000,A=0.000000)" :AttenuationMaxRangeMultiplier 1.0 :AttenuationIntensityMultiplier 1.0)))

(event EventOnExploded
  (params (HitActors :type Array%20of%20Actor%20Object%20References) (HitResults :type Array%20of%20Hit%20Result%20Structures))
  (bind _stoplocation (Variables|Effects|GetStopLocation))
  (bind _stoprotz (Variables|Effects|GetStopRotZ))
  ;[1 1 1] Impact and Explosion Stuff
  (Networking|CanExecuteCosmeticEvents
    (:True
      (Development|SetHiddeninGame (Variables|Projectile_Rocket|GetMesh) true false)
      (Utilities|FlowControl|Sequence
        (:then_0
          (Components|Activation|Deactivate (Variables|Default|GetNS_RocketTrail_DrunkenQuad_01)))
        (:then_1
          (Components|Activation|SetActive (Variables|Effects|GetRocketTrailPS) false false)
          ;;[1 1 1] SingleLineCheckForWorldObjectsSurfaceType
          (if (Utilities|Enum|Equal(Enum) (Variables|Default|GetSurfaceType) "SurfaceType8")
            (bind _returnvalue (Effects|Components|ParticleSystem|SpawnFXAtLocation "/Game/Effects/Explosion_Update/NS_GrenadeLauncher_Explosion_Asteria_Water.NS_GrenadeLauncher_Explosion_Asteria_Water" _stoplocation _stoprotz))
            (Rendering|SetRenderCustomDepth _returnvalue true)
            (Audio|PlaySoundatLocation (Variables|Sounds|GetRocket_Projectile_Explosion_Water_Sound) (Variables|Effects|GetStopLocation))
            (else
              (bind _returnvalue_2 (Effects|Components|ParticleSystem|SpawnFXAtLocation "/Game/Effects/Explosion_Update/NS_GrenadeLauncher_Explosion_Asteria.NS_GrenadeLauncher_Explosion_Asteria" _stoplocation _stoprotz))
              (Rendering|SetRenderCustomDepth _returnvalue_2 true)
              (Audio|PlaySoundatLocation (Variables|Sounds|GetRocket_Projectile_Explosion_Sound) (Variables|Effects|GetStopLocation))))
          (Rendering|SetVisibility (Variables|Projectile_Rocket|GetMesh) false false)
          (Class|BPSurfaceTypeSoundComponent|PlaySound (Variables|Default|GetBP_SurfaceTypeSoundComponent) (Transformation|GetActorLocation))
          ;[1 1 1] Shake that camera
          (Audio|Components|Audio|FadeOut (Variables|Default|GetRocket_AudioComponent) 1 0.0 "Linear" :!class /Script/Engine.AudioComponent)
          (GameplayCue|ExecuteGameplayCue%28Burst%29Local self (Variables|Default|GetFeedbackCue) (Ability|GameplayCue|MakeGameplayCueParameters 0.0)) ; Cam shake + controller vibration
        )))))

(event EventDestroyed
  ;[1 1 1] Impact and Explosion Stuff
  (Networking|CanExecuteCosmeticEvents))

(fn ConstructionScript
  (Effects|Components|ParticleSystem|SetTemplate (Variables|Default|GetRibbon_Trail_PSC) (Variables|Effects|GetRibbon_Trail_PS))
  (Effects|Components|ParticleSystem|SetTemplate (Variables|Default|GetExplosion_Generic_PSC) (Variables|Effects|GetExplosion_Generic_PS))
  (Effects|Components|ParticleSystem|SetTemplate (Variables|Default|GetExplosion_Flesh_Damage_PSC) (Variables|Effects|GetExplosion_Flesh_Damage_PS))
  (Transformation|SetRelativeScale3D (Variables|Default|GetCollisionMesh) (Variables|Projectile_Rocket|GetRocketScale)))
