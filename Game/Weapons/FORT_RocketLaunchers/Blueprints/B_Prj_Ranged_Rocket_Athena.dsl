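; Rocket launcher projectile for Athena that flies with a spinning mesh and trail, then explodes on stop or fuse timeout.
; A delayed timer enables water contact, detonating the rocket if already submerged.
; Explosion or kill-disable runs cosmetic-gated water or land FX chosen by surface type, plus sound, radial force, post-process and feedback cue.
; Hits on bounce-tagged components double velocity, and server-side overlap broadcasts a rocket-riding quest message.
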(event EventBeginPlay
  ;[1 1 1] EVENT -> Begin Play
  (Utilities|FlowControl|Sequence
    (:then_0
      ;;[1 1 1] Timer to enable touching water to cause an explosion
      (if (> (Variables|Effects|GetWaterInteractionDelay) 0.0)
        (Utilities|Time|SetTimerbyEvent (EventDispatchers|CreateEvent :!event Custom|RocketWaterDelay) (Variables|Effects|GetWaterInteractionDelay))
        (else
          (Variables|Projectile|SetTouchWaterBehavior "StopIfStopSimulatingOnHit"))))
    (:then_1
      ;;[1 1 1] Spawn attached Rocket Trail
      (Utilities|Time|SetTimerbyFunctionName self "FuseTimerMax" (Variables|Default|GetFuseTimer))
      (Networking|CanExecuteCosmeticEvents
        (:True
          (bind _returnvalue (Effects|SpawnFXSystemAttached :FXSystemAsset (Variables|Effects|RocketTrail|GetRocket%20Trail%20Asset) :AttachToComponent (Variables|Collision|GetCapsuleComponent) :AttachPointName "None" :Rotation (Math|Rotator|MakeRotator (Math|Random|RandomFloatinRange 0.0 360)) :Scale "1.000000,1.000000,1.000000" :LocationType "KeepRelativeOffset" :bAutoDestroy true :PoolingMethod "None" :bAutoActivate true))
          (bind _asniagara_particle_system_component (Utilities|Casting|CastToNiagaraComponent _returnvalue)
            (:then
              (Variables|Default|SetIsNiagaraEnabled true)
              (bind _output_get (Variables|Default|SetRocketTrailPS_Niagara _asniagara_particle_system_component)))
            (:CastFailed
              (bind _ascascade_particle_system_component (Utilities|Casting|CastToParticleSystemComponent _returnvalue))
              (bind _output_get_2 (Variables|Effects|RocketTrail|SetRocketTrailPS _ascascade_particle_system_component))))
          (Rendering|SetRenderCustomDepth _output_get true))))))

(event EventTick
  (params (DeltaSeconds :type Float%20%28single-precision%29))
  ;[1 1 1] EVENT -> Tick: spin the rocket and particle trail
  (Variables|Effects|SetTick%20Delta DeltaSeconds)
  (Networking|CanExecuteCosmeticEvents
    (:True
      (if (Variables|Default|GetIsNiagaraEnabled)
        (if (Components|Activation|IsActive (Variables|Default|GetRocketTrailPS_Niagara))
          _)
        (elif (Components|Activation|IsActive (Variables|Effects|RocketTrail|GetRocketTrailPS))
          _))
      (Variables|Effects|RocketTrail|SetRocketSpin (Math|Rotator|MakeRotator (* (Variables|Effects|GetTick%20Delta) (Variables|Effects|RocketTrail|GetRocketSpinRate))))
      (Transformation|AddLocalRotation (Variables|Projectile_Rocket|GetMesh) (Variables|Effects|RocketTrail|GetRocketSpin) false false))))

(event EventOnTouched
  (params (OtherActor :type Actor%20Object%20Reference) (OtherComp :type Primitive%20Component%20Object%20Reference) (HitResult :type Hit%20Result%20Structure%20%28by%20ref%29) (bIsOverlap :type Boolean))
  (bind (_bblockinghit _binitialoverlap _time _distance _location _impactpoint _normal _impactnormal _physmat _hitactor _hitcomponent) (Collision|BreakHitResult HitResult))
  ;[1 1 1] EVENT -> Touched: Bounce off "NoBounce"
  (if (Components|ComponentHasTag _hitcomponent "ShouldCauseBounce")
    (Variables|/Script/Engine.MovementComponent|SetVelocity (* (Variables|/Script/Engine.MovementComponent|GetVelocity (Variables|Movement|GetProjectileMovementComponent)) 2.0) (Variables|Movement|GetProjectileMovementComponent))
    (else
      (Variables|Projectile|SetbStopSimulatingOnHit true))))

(event Custom|RocketWaterDelay
  ;[1 1 1] Delay for rocket to explode in Water
  (if (and (Utilities|IsValid (Variables|/Script/Engine.MovementComponent|GetUpdatedComponent (Variables|Movement|GetProjectileMovementComponent)) :!class /Script/Engine.KismetSystemLibrary) (not (Delivery|IsBeingKilled)))
    (Variables|Projectile|SetTouchWaterBehavior "StopIfStopSimulatingOnHit")
    (if (Water|IsinWater (Variables|Water|GetWaterInteractionComponent))
      (CallFunction|FuseTimerMax))))

(event EventOnStop
  (params (Hit :type Hit%20Result%20Structure%20%28by%20ref%29))
  (bind (_bblockinghit _binitialoverlap _time _distance _location _impactpoint _normal _impactnormal _physmat _hitactor) (Collision|BreakHitResult Hit))
  ;[1 1 1] EVENT -> Stop: Impact and Explosion Stuff
  (Variables|Effects|Explosion|Decal|SetDecalLocation _impactpoint)
  (Variables|Effects|SetStopLocation (+ _impactpoint (* _impactnormal (Variables|Effects|Explosion|GetExplosion%20Impact%20Offset))))
  (Variables|Effects|SetStopRotZ (Math|Rotator|MakeRotfromZ _impactnormal))
  (Utilities|IsValid _physmat
    (:Is%20Valid
      (Variables|Effects|SetSurfaceType (Variables|/Script/PhysicsCore.PhysicalMaterial|GetSurfaceType _physmat)))
    (:Is%20Not%20Valid
      (Variables|Effects|SetSurfaceType "SurfaceType_Default")))
  (Class|BPSurfaceTypeSoundComponent|SetSurfaceType (Variables|Sound|GetBP_SurfaceTypeSoundComponent) (Variables|Effects|GetSurfaceType) _hitactor)
  (if (not (Delivery|IsBeingKilled))
    (Delivery|DoExplosion)
    (Fortnite|ShowSoundIndicatorLocal :Location (Transformation|GetWorldLocation (Variables|Projectile_Rocket|GetMesh)) :MaxAudibleDistance 5e+03 :IndicatorType "Gunshot" :Instigator self :Tint "(R=0.000000,G=0.000000,B=0.000000,A=0.000000)" :AttenuationMaxRangeMultiplier 1.0 :AttenuationIntensityMultiplier 1.0)))

(event Custom|FuseTimerMax
  (bind _mesh (Variables|Projectile_Rocket|GetMesh))
  (bind _returnvalue (Transformation|GetWorldLocation _mesh))
  ;[1 1 1] EVENT -> Stop: Impact and Explosion Stuff
  (Variables|Effects|Explosion|SetTimerMaxReached? true)
  (Variables|Effects|SetStopLocation _returnvalue)
  (if (not (Delivery|IsBeingKilled))
    (Delivery|DoExplosion)
    (Fortnite|ShowSoundIndicatorLocal :Location _returnvalue :MaxAudibleDistance 5e+03 :IndicatorType "Gunshot" :Instigator self :Tint "(R=0.000000,G=0.000000,B=0.000000,A=0.000000)" :AttenuationMaxRangeMultiplier 1.0 :AttenuationIntensityMultiplier 1.0)))

(event EventOnExploded
  (params (HitActors :type Array%20of%20Actor%20Object%20References) (HitResults :type Array%20of%20Hit%20Result%20Structures))
  (bind _surfacetype (Variables|Effects|GetSurfaceType))
  (bind _returnvalue (Math|Conversions|ToInteger(Byte) _surfacetype))
  (bind _stoplocation (Variables|Effects|GetStopLocation))
  (bind _rockettrailps_niagara (Variables|Default|GetRocketTrailPS_Niagara))
  (bind _rockettrailps (Variables|Effects|RocketTrail|GetRocketTrailPS))
  ;[1 0.5861 0.08121] EVENT -> Exploded
  (Utilities|FlowControl|DoOnce false
    (:Completed
      (Networking|CanExecuteCosmeticEvents
        (:True
          (Development|SetHiddeninGame (Variables|Projectile_Rocket|GetMesh) true true)
          (Utilities|IsValid _rockettrailps
            (:Is%20Valid
              (Components|Activation|Deactivate _rockettrailps))
            (:Is%20Not%20Valid
              _))
          (Utilities|IsValid _rockettrailps_niagara
            (:Is%20Valid
              (Components|Activation|Deactivate _rockettrailps_niagara))
            (:Is%20Not%20Valid
              _))
          (if (or (Utilities|Enum|Equal(Enum) _surfacetype "SurfaceType8") (Water|IsinWater (Variables|Water|GetWaterInteractionComponent)))
            ;;[0.6757 0 1] Spawn Water FX and Sound
            (Class|BPSurfaceTypeSoundComponent|SetSurfaceType (Variables|Sound|GetBP_SurfaceTypeSoundComponent) "SurfaceType8" 0)
            (Effects|Components|ParticleSystem|SpawnFXAtLocation (Variables|Effects|Explosion|GetExplosion%20Asset_Water) (Variables|Effects|GetStopLocation) (Variables|Effects|GetStopRotZ))
            (Utilities|SpawnFNRadialForce (Utilities|StopLocationTransform) (Variables|PhysicsFields|GetPlayrate) (Variables|PhysicsFields|GetRadius) (Variables|PhysicsFields|GetStrength) (Variables|PhysicsFields|GetTime%20Decay))
            (Audio|PlaySoundatLocation (Variables|Sounds|GetRocket_Projectile_Explosion_Water_Sound) (Variables|Effects|GetStopLocation))
            (else
              ;;[0.2043 0 1 1 30] Niagara Explosion Branch
              (Effects|Components|ParticleSystem|SpawnFXAtLocation (Variables|Effects|Explosion|GetExplosion%20Asset) _stoplocation (Variables|Effects|GetStopRotZ))
              ;[1 1 1] PhysSurface Interaction for explosion Niagara FX
              (bind _checkedphysicalsurfaceid (CallFunction|PhysicalSurfaceCheck _returnvalue))
              (if _checkedphysicalsurfaceid
                (bind _returnvalue_2 (Effects|Components|ParticleSystem|SpawnFXAtLocation (Variables|Default|GetFX_Physical%20Surface) _stoplocation))
                (Effects|Components|ParticleSystem|SetIntParameter _returnvalue_2 "PhysicalSurface" _returnvalue)
                (Effects|Components|ParticleSystem|SetFloatParameter _returnvalue_2 "Global Scale" 0.9))
              (Utilities|SpawnFNRadialForce (Utilities|StopLocationTransform) (Variables|PhysicsFields|GetPlayrate) (Variables|PhysicsFields|GetRadius) (Variables|PhysicsFields|GetStrength) (Variables|PhysicsFields|GetTime%20Decay))
              (Audio|PlaySoundatLocation (Variables|Sounds|GetRocket_Projectile_Explosion_Sound) (Variables|Effects|GetStopLocation))))
          (Class|BPSurfaceTypeSoundComponent|PlaySound (Variables|Sound|GetBP_SurfaceTypeSoundComponent) (Variables|Effects|GetStopLocation))
          ;;[0.6661 0.1729 1] Shake that camera / Post Process
          (Utilities|FlowControl|Sequence
            (:then_0
              (if (<= (Math|Vector|VectorLength (- (Variables|Effects|GetStopLocation) (Transformation|GetWorldLocation (Variables|/Script/Engine.PlayerCameraManager|GetTransformComponent (Game|GetPlayerCameraManager 0))))) (Variables|/Script/Engine.PostProcessComponent|GetBlendRadius (Variables|Default|GetPP_Explode)))
                (bind (_direction _ppfade) (Timeline|PPFader PlayFromStart)
                  (:Update
                    (Variables|/Script/Engine.PostProcessComponent|SetBlendWeight _ppfade (Variables|Default|GetPP_Explode))))))
            (:then_1
              (GameplayCue|ExecuteGameplayCue%28Burst%29Local self (Variables|Effects|GetFeedbackCue) (Ability|GameplayCue|MakeGameplayCueParameters 0.0)) ; Cam shake + controller vibration
              (bind _rocket_audiocomponent (Variables|Default|GetRocket_AudioComponent)
                (:then
                  (Audio|Components|Audio|FadeOut _rocket_audiocomponent 0.15 0.0 "Linear" :!class /Script/Engine.AudioComponent))
                (:else)))))))))

(event EventOnDisableForKill
  (bind _surfacetype (Variables|Effects|GetSurfaceType))
  (bind _returnvalue (Math|Conversions|ToInteger(Byte) _surfacetype))
  (bind _stoplocation (Variables|Effects|GetStopLocation))
  (bind _rockettrailps_niagara (Variables|Default|GetRocketTrailPS_Niagara))
  (bind _rockettrailps (Variables|Effects|RocketTrail|GetRocketTrailPS))
  ;[1 0.5861 0.08121] EVENT -> Exploded
  (Utilities|FlowControl|DoOnce false
    (:Completed
      (Networking|CanExecuteCosmeticEvents
        (:True
          (Development|SetHiddeninGame (Variables|Projectile_Rocket|GetMesh) true true)
          (Utilities|IsValid _rockettrailps
            (:Is%20Valid
              (Components|Activation|Deactivate _rockettrailps))
            (:Is%20Not%20Valid
              _))
          (Utilities|IsValid _rockettrailps_niagara
            (:Is%20Valid
              (Components|Activation|Deactivate _rockettrailps_niagara))
            (:Is%20Not%20Valid
              _))
          (if (or (Utilities|Enum|Equal(Enum) _surfacetype "SurfaceType8") (Water|IsinWater (Variables|Water|GetWaterInteractionComponent)))
            ;;[0.6757 0 1] Spawn Water FX and Sound
            (Class|BPSurfaceTypeSoundComponent|SetSurfaceType (Variables|Sound|GetBP_SurfaceTypeSoundComponent) "SurfaceType8" 0)
            (Effects|Components|ParticleSystem|SpawnFXAtLocation (Variables|Effects|Explosion|GetExplosion%20Asset_Water) (Variables|Effects|GetStopLocation) (Variables|Effects|GetStopRotZ))
            (Utilities|SpawnFNRadialForce (Utilities|StopLocationTransform) (Variables|PhysicsFields|GetPlayrate) (Variables|PhysicsFields|GetRadius) (Variables|PhysicsFields|GetStrength) (Variables|PhysicsFields|GetTime%20Decay))
            (Audio|PlaySoundatLocation (Variables|Sounds|GetRocket_Projectile_Explosion_Water_Sound) (Variables|Effects|GetStopLocation))
            (else
              ;;[0.2043 0 1 1 30] Niagara Explosion Branch
              (Effects|Components|ParticleSystem|SpawnFXAtLocation (Variables|Effects|Explosion|GetExplosion%20Asset) _stoplocation (Variables|Effects|GetStopRotZ))
              ;[1 1 1] PhysSurface Interaction for explosion Niagara FX
              (bind _checkedphysicalsurfaceid (CallFunction|PhysicalSurfaceCheck _returnvalue))
              (if _checkedphysicalsurfaceid
                (bind _returnvalue_2 (Effects|Components|ParticleSystem|SpawnFXAtLocation (Variables|Default|GetFX_Physical%20Surface) _stoplocation))
                (Effects|Components|ParticleSystem|SetIntParameter _returnvalue_2 "PhysicalSurface" _returnvalue)
                (Effects|Components|ParticleSystem|SetFloatParameter _returnvalue_2 "Global Scale" 0.9))
              (Utilities|SpawnFNRadialForce (Utilities|StopLocationTransform) (Variables|PhysicsFields|GetPlayrate) (Variables|PhysicsFields|GetRadius) (Variables|PhysicsFields|GetStrength) (Variables|PhysicsFields|GetTime%20Decay))
              (Audio|PlaySoundatLocation (Variables|Sounds|GetRocket_Projectile_Explosion_Sound) (Variables|Effects|GetStopLocation))))
          (Class|BPSurfaceTypeSoundComponent|PlaySound (Variables|Sound|GetBP_SurfaceTypeSoundComponent) (Variables|Effects|GetStopLocation))
          ;;[0.6661 0.1729 1] Shake that camera / Post Process
          (Utilities|FlowControl|Sequence
            (:then_0
              (if (<= (Math|Vector|VectorLength (- (Variables|Effects|GetStopLocation) (Transformation|GetWorldLocation (Variables|/Script/Engine.PlayerCameraManager|GetTransformComponent (Game|GetPlayerCameraManager 0))))) (Variables|/Script/Engine.PostProcessComponent|GetBlendRadius (Variables|Default|GetPP_Explode)))
                (Timeline|PPFader PlayFromStart)))
            (:then_1
              (GameplayCue|ExecuteGameplayCue%28Burst%29Local self (Variables|Effects|GetFeedbackCue) (Ability|GameplayCue|MakeGameplayCueParameters 0.0)) ; Cam shake + controller vibration
              (bind _rocket_audiocomponent (Variables|Default|GetRocket_AudioComponent)
                (:then
                  (Audio|Components|Audio|FadeOut _rocket_audiocomponent 0.15 0.0 "Linear" :!class /Script/Engine.AudioComponent))
                (:else)))))))))

(event EventDestroyed
  (bind _rockettrailps (Variables|Effects|RocketTrail|GetRocketTrailPS))
  ;[1 1 1] EVENT -> Destroyed: Deactivating Rocket Trail Particle systems on component destroyed
  (Networking|CanExecuteCosmeticEvents
    (:True
      (Utilities|IsValid _rockettrailps
        (:Is%20Valid
          _)
        (:Is%20Not%20Valid
          (Utilities|IsValid (Variables|Default|GetRocketTrailPS_Niagara)
            (:Is%20Valid
              _))))
      (Components|Activation|Deactivate _rockettrailps)
      (bind _rocket_audiocomponent (Variables|Default|GetRocket_AudioComponent)
        (:then
          (Audio|Components|Audio|FadeOut _rocket_audiocomponent 0.15 0.0 "Linear" :!class /Script/Engine.AudioComponent))
        (:else)))))

(event OnComponentBeginOverlap(RocketRidingCollision)
  (params (OverlappedComponent :type Primitive%20Component%20Object%20Reference) (OtherActor :type Actor%20Object%20Reference) (OtherComp :type Primitive%20Component%20Object%20Reference) (OtherBodyIndex :type Integer) (bFromSweep :type Boolean) (SweepResult :type Hit%20Result%20Structure%20%28by%20ref%29))
  (bind _playerridingrocket (Variables|Quest|GetPlayerRidingRocket))
  ;[0 0 0] EVENT -> On Riding Rocket Collision overlapped by a player, broadcast a quest message for riding a rocket
  (Networking|SwitchHasAuthority
    (:Authority
      (bind _asfort_player_pawn (Utilities|Casting|CastToFortPlayerPawn OtherActor))
      (Utilities|FlowControl|Delay 0.2)
      (if (not (Delivery|IsBeingKilled))
        (Utilities|IsValid _playerridingrocket
          (:Is%20Valid
            (if (not (== _playerridingrocket _asfort_player_pawn))
              (Variables|Quest|SetPlayerRidingRocket _asfort_player_pawn)))
          (:Is%20Not%20Valid
            (Variables|Quest|SetPlayerRidingRocket _asfort_player_pawn)))
        (Utilities|FlowControl|DoOnce false
          (:Completed
            (CustomVerb|BroadcastCustomVerbMessage (CustomVerb|MakeCustomVerbMessage (Variables|Quest|GetPlayerRidingRocket) "(TagName=\"CustomVerb.PlayerRidingRocket\")"))))))))

(fn GetFXAsset
  (params (AssetStruct :type Fort%20Portable%20Soft%20Particles%20Structure))
  (bind (_fxtype _niagaraversion _cascadeversion) (Utilities|Struct|BreakFortPortableSoftParticles AssetStruct))
  (return (select (Effects|IsNiagaraFXTypeEnabled _fxtype) (Utilities|ResolveSoftReference _niagaraversion) (Utilities|ResolveSoftReference _cascadeversion))))

(fn PhysicalSurfaceCheck
  (params (Physical%20Surface%20ID :type Integer))
  (return (Math|Boolean|ORBoolean (== Physical%20Surface%20ID 30) (== Physical%20Surface%20ID 26) (== Physical%20Surface%20ID 23) (== Physical%20Surface%20ID 24))))

(fn ConstructionScript
  (Effects|Components|ParticleSystem|SetTemplate (Variables|Effects|RocketTrail|GetRibbon_Trail_PSC) (Variables|Effects|RocketTrail|GetRibbon_Trail_PS))
  (Effects|Components|ParticleSystem|SetTemplate (Variables|Default|GetExplosion_Flesh_Damage_PSC) (Variables|Effects|Explosion|Surfaces|GetExplosion_Flesh_Damage_PS))
  (Transformation|SetRelativeScale3D (Variables|Default|GetCollisionMesh) (Variables|Projectile_Rocket|GetRocketScale)))
