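; Generic rocket launcher weapon Blueprint that adds launcher-specific behavior to the inherited muzzle flash effect.
; After the parent muzzle flash event runs, it toggles the active state of the RearMuzzle component.
; If the weapon mesh uses the front-action launcher template skeleton, it sets the spawned muzzle system's relative rotation to the Rotation Offset variable.
; The construction script only calls the parent.
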(event EventMuzzleFlashFX
  (params (Persistent%20Fire :type Boolean))
  (|CallParentFunction Persistent%20Fire)
  (Components|Activation|ToggleActive (Variables|Default|GetRearMuzzle))
  (if (Utilities|Equal(Object) (Variables|/Script/Engine.SkeletalMesh|GetSkeleton (Components|SkeletalMesh|GetSkeletalMeshAsset (Variables|Visual|GetWeaponMesh))) "/Game/Weapons/Skeleton_Templates/Launcher_FrontAction_Template_Skeleton.Launcher_FrontAction_Template_Skeleton")
    (Transformation|SetRelativeRotation (Variables|Effects|Muzzle|GetSpawned%20Muzzle%20System) (Variables|Default|GetRotation%20Offset) false false)))

(fn ConstructionScript
  (|CallParentFunction))
