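; Generic Athena rocket launcher weapon Blueprint that adds rear muzzle effects and controls the visible rocket on the weapon mesh.
; The construction script picks the Niagara or older particle rear-muzzle system based on whether Niagara Projectile FX are enabled.
; Muzzle flash calls the parent, then attaches that system at the rear muzzle with an offset.
; On cosmetic-capable instances, equipping with empty ammo hides the Rocket bone, and reload start unhides it.
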(event EventMuzzleFlashFX
  (params (Persistent%20Fire :type Boolean))
  (bind _rearmuzzlefxsystem (Variables|Default|GetRearMuzzleFXSystem))
  (|CallParentFunction false)
  ;[1 1 1] This replaces a component.
  (bind _returnvalue (Effects|SpawnFXSystemAttached _rearmuzzlefxsystem (Variables|Visual|GetWeaponMesh) "None" (Variables|/Script/Engine.SceneComponent|GetRelativeLocation (Variables|Default|GetRearMuzzleLoc)) "0,-90.000000, 0"))
  (Utilities|Casting|CastToNiagaraSystem _rearmuzzlefxsystem)
  (Transformation|AddLocalOffset _returnvalue (Variables|Default|GetRearMuzzleOffsetMultiplier) false false))

(event FortWeapon|EventOnEquippedAndAttached
  ; Handle missile visibility based on ammo
  (Networking|CanExecuteCosmeticEvents
    (:True
      (EventDispatchers|BindEventtoOnReloadStarted (delegate Custom|OnLocalReloadStarted_Event))
      (if (<= (FortWeapon|GetLocalRemainingAmmo) 0)
        (Components|SkinnedMesh|HideBonebyName (Variables|Visual|GetWeaponMesh) "Rocket" "PBO_None")))))

(event Custom|OnLocalReloadStarted_Event
  (params (ReloadTime :type Float%20%28single-precision%29) (ReloadType :type EFortWeaponReloadType%20Enum :default "ReloadWholeClip"))
  ; Handle missile visibility based on ammo
  (Components|SkinnedMesh|UnHideBonebyName (Variables|Visual|GetWeaponMesh) "Rocket"))

(fn ConstructionScript
  (|CallParentFunction)
  (Variables|Default|SetRearMuzzleFXSystem (select (Effects|IsNiagaraFXTypeEnabled "Projectile") "/Game/Weapons/FORT_RocketLaunchers/Effects/Muzzle/NS_RocketLauncher_RearMuzzle_Athena.NS_RocketLauncher_RearMuzzle_Athena" "/Game/Weapons/FORT_RocketLaunchers/Effects/Muzzle/P_RocketLauncher_RearMuzzle_01.P_RocketLauncher_RearMuzzle_01")))
