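; Military rocket launcher weapon Blueprint for Athena that drives muzzle, back-blast and reload effects for a four-barrel launcher.
; On firing, it activates the back muzzle effect, and a DoN counter indexes offsets that move front and back effects between barrels.
; At the matching reload stage, a delay precedes activation of the magazine-eject particles.
; Their collision event is bound to a DoOnce-gated SoundFX handler that only prints debug text.
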(event EventMuzzleFlashFX
  (params (Persistent%20Fire :type Boolean))
  (|CallParentFunction Persistent%20Fire)
  (Components|Activation|Activate (Variables|Default|GetMuzzleBack) true)
  (bind _counter (Utilities|FlowControl|DoN 4)
    (:Exit
      (Transformation|SetRelativeLocation (Variables|Effects|Muzzle|GetSpawned%20Muzzle%20System) (Utilities|Select :Index _counter :Option%200 "0, 0, 0" :Option%201 "0, 0, 0" :Option%202 "0.000000,-11.081369,0.000000" :Option%203 "0.000000,0.000000,-10.683455" :Option%204 "0.000000,-11.081369,-10.985550") false false)
      (Transformation|SetRelativeLocation (Variables|Default|GetMuzzleBack) (Utilities|Select :Index _counter :Option%200 "0, 0, 0" :Option%201 "-24.600000,3.730000,2.190000" :Option%202 "-24.600000,-4.700000,2.190000" :Option%203 "-24.600000,3.730000,-5.570000" :Option%204 "-24.600000,-4.700000,-5.570000") false false)
      (Utilities|FlowControl|Sequence
        (:then_0
          (if (>= _counter 4)
            _))
        (:then_1
          _))
      (Transformation|SetRelativeScale3D (Variables|Effects|Muzzle|GetSpawned%20Muzzle%20System) "1.000000,1.000000,1.000000"))))

(event EventOnPlayReloadFX
  (params (ReloadStage :type EFortReloadFXState%20Enum :default "ReloadStart"))
  ;[1 1 1] Magazine ejecting and bouncing animation while reloading
  (EventDispatchers|BindEventtoOnParticleCollide (Variables|Default|GetP_M202_Muzzle_MagazineEject) (delegate Custom|SoundFX))
  (if (and (Utilities|Enum|Equal(Enum) ReloadStage (Variables|Default|GetReloadState)) (FortWeapon|IsReloading))
    (Utilities|FlowControl|Delay 0.55)
    (Development|PrintString (Utilities|String|EnumtoString ReloadStage) true true "(R=0.000000,G=0.660000,B=1.000000,A=1.000000)" 8.0)
    (Components|Activation|Activate (Variables|Default|GetP_M202_Muzzle_MagazineEject) false)))

(event Custom|SoundFX
  (params (EventName :type Name) (EmitterTime :type Float%20%28single-precision%29) (ParticleTime :type Integer) (Location :type Vector) (Velocity :type Vector) (Direction :type Vector) (Normal :type Vector) (BoneName :type Name) (PhysMat :type Physical%20Material%20Object%20Reference))
  ;[1 1 1] Bound a particle collision event For SFX
  (Utilities|FlowControl|DoOnce false
    (:Completed
      (Development|PrintString "ActiveSFX")
      (CallFunction|ResetDoOnce))))

(event Custom|ResetDoOnce
  ;[1 1 1] Bound a particle collision event For SFX
  (Utilities|FlowControl|DoOnce false
    (:Completed
      (Development|PrintString "ActiveSFX")
      (CallFunction|ResetDoOnce))))

(fn ConstructionScript
  (|CallParentFunction))
