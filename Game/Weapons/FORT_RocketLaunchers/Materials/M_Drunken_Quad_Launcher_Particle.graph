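# Default Lit material for the Drunken Quad Launcher rocket launcher, flagged for use on Niagara mesh particles.
# It holds three texture samples of the launcher's weapon textures: D as color, N as a normal map and S as masks.
# No math, parameters or links to material outputs appear in the graph.
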Material /Game/Weapons/FORT_RocketLaunchers/Materials/M_Drunken_Quad_Launcher_Particle
graph {
 properties {bUsedWithNiagaraMeshParticles=true, pixelDepthOffsetMode=PDOM_Legacy, usedShadingModels=Default Lit}
 nodes [
  TextureSample() {samplerType=SAMPLERTYPE_Normal, texture=/Game/Weapons/FORT_RocketLaunchers/Textures/T_Weapon_Drunken_Quad_Launcher_N}
  TextureSample() {texture=/Game/Weapons/FORT_RocketLaunchers/Textures/T_Weapon_Drunken_Quad_Launcher_D}
  TextureSample() {samplerType=SAMPLERTYPE_Masks, texture=/Game/Weapons/FORT_RocketLaunchers/Textures/T_Weapon_Drunken_Quad_Launcher_S}
 ]
}
