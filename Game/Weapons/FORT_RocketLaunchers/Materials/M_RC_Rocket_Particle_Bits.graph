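# Default Lit material for mesh particles, apparently the debris bits of the RC rocket, judging by its name.
# It holds three texture samples that reuse the RC rocket launcher's diffuse, normal and specular mask maps.
# The graph shows no math, parameters or visible connections from the samples to the material outputs.
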Material /Game/Weapons/FORT_RocketLaunchers/Materials/M_RC_Rocket_Particle_Bits
graph {
 properties {bUsedWithMeshParticles=true, pixelDepthOffsetMode=PDOM_Legacy, usedShadingModels=Default Lit}
 nodes [
  TextureSample() {samplerType=SAMPLERTYPE_Normal, texture=/Game/Weapons/FORT_RocketLaunchers/Textures/T_RC_RocketLauncher_N}
  TextureSample() {texture=/Game/Weapons/FORT_RocketLaunchers/Textures/T_RC_RocketLauncher_D}
  TextureSample() {samplerType=SAMPLERTYPE_Masks, texture=/Game/Weapons/FORT_RocketLaunchers/Textures/T_RC_RocketLauncher_S}
 ]
}
