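# Translucent, unlit, two-sided mesh-particle material for a rocket launcher refraction cone, producing distortion by index of refraction.
# A dynamic-parameter Refraction Amount is blended from neutral by vertex alpha, then masked by an adjustable Fresnel.
# A swirl-noise normal map pans diagonally at a particle-driven speed to break up the distortion.
# Its tiling uses U Tile on both axes, leaving the declared V Tile unused.
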Material /Game/Weapons/FORT_RocketLaunchers/Materials/M_RocketLauncher_RefractionCone_01
graph {
 properties {
  bUsedWithMeshParticles=true
  blendMode=BLEND_Translucent
  pixelDepthOffsetMode=PDOM_Legacy
  previewMesh=/Game/Weapons/FORT_RocketLaunchers/Mesh/St_RocketLauncher_RefractionCone_01
  refractionMethod=RM_IndexOfRefraction
  shadingModel=MSM_Unlit
  twoSided=true
  usedShadingModels=Unlit
 }
 nodes [
  Constant()
  LinearInterpolate(
   B=LinearInterpolate(
    B=DynamicParameter() {defaultValue=<rgba 1.2 1 4 0.5>, paramNames [Refraction Amount, Panner Speed, U Tile, V Tile]}.Refraction Amount
    Alpha=VertexColor().A
   ) {constA=1}
   Alpha=MaterialFunctionCall(
    Brightness=ScalarParameter() {defaultValue=1, parameterName=Fresnel Brightness}
    Exponent=ScalarParameter() {defaultValue=3, parameterName=Fresnel Exponent}
   ) {materialFunction=&MF_Fresnel_FX}.Fresnel
  ) {constA=1}
  n1 = TextureCoordinate()
  n2 = DynamicParameter() {defaultValue=<rgba 1.2 1 4 0.5>, paramNames [Refraction Amount, Panner Speed, U Tile, V Tile]}
  n3 = AppendVector(A=@n2.U Tile, B=@n2.U Tile)
  n4 = Time()
  n5 = DynamicParameter() {defaultValue=<rgba 1.2 1 4 0.5>, paramNames [Refraction Amount, Panner Speed, U Tile, V Tile]}
  TextureSample(UVs=Panner(Coordinate=Multiply(A=@n1, B=@n3), Time=Multiply(A=@n4, B=@n5.Panner Speed)) {bFractionalPart=true, speedX=0.8, speedY=-3.2}) {samplerType=SAMPLERTYPE_Normal, texture=/Game/Global/Textures/Default/Noise/T_SwirlNoise_N}
 ]
}
