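# Physics asset for the RC_Rocket_Export mesh among the rocket launcher weapons, giving it simple physics shapes.
# It defines two bodies, Handle and Rocket, each approximated by a single rotated, offset box.
# No constraints join the two bodies.
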PhysicsAsset /Game/Weapons/FORT_RocketLaunchers/Mesh/RC_Rocket_Export_Physics
bodies {
 Handle [{shapeType=Box, extentX=21.6663, extentY=34.4527, extentZ=102.4025, center=<0.3318 2.0249 11.1041>, rotation=<rot 0 90 -89.6491>}]
 Rocket [{shapeType=Box, extentX=25, extentY=111, extentZ=25, center=<5.9243 0 0>, rotation=<rot 0 -90 0>}]
}
constraints []
