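# Physics asset for the RC rocket mesh in the rocket launcher weapon set, giving the rocket a simple physics shape.
# A single body on the Rocket bone holds one elongated box, yawed a quarter turn and nudged so its long axis runs along the bone's X axis.
# It has no constraints, so there is no articulation between parts.
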PhysicsAsset /Game/Weapons/FORT_RocketLaunchers/Mesh/RC_Rocket_Only_Export_Physics
bodies {Rocket [{shapeType=Box, extentX=39.5, extentY=109, extentZ=39.5, center=<5.2526 0.0002 0>, rotation=<rot 0 -90 0>}]}
constraints []
