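# Physics asset for the SK_Empty_Weapon skeletal mesh in the rocket launcher weapon folder, judging by its name and path.
# It holds a single body on the Handle bone, made of one very small, offset and rotated capsule.
# It has no constraints, so it defines no jointed ragdoll setup, only that minimal collision shape.
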PhysicsAsset /Game/Weapons/FORT_RocketLaunchers/Mesh/SK_Empty_Weapon_PhysicsAsset
bodies {Handle [{shapeType=Capsule, length=0.505, radius=0.505, center=<-2.9112 -0.4541 7.8643>, rotation=<rot 0 174.8373 -47.4522>}]}
constraints []
