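# Physics asset for the M202 military rocket launcher skeletal mesh, defining simple collision for the launcher and its flaps.
# A single box body covers the handle bone, and one box each covers the front and back dynamic flap bones.
# Each flap is constrained to the handle with both swing axes locked, leaving twist as its only angular movement, like a hinge.
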PhysicsAsset /Game/Weapons/FORT_RocketLaunchers/Mesh/SK_Military_M202_Launcher_PhysicsAsset
bodies {
 handle [{shapeType=Box, extentX=38.238, extentY=92.2436, extentZ=25.655, center=<-3.8959 3.2763 18.4262>, rotation=<rot 0 0 -0.0003>}]
 m202_dyn_flap_back [{shapeType=Box, extentX=25.5806, extentY=5.0357, extentZ=18.2608, center=<-0.0098 -4.6722 -7.4378>, rotation=<rot 0 0 29.9999>}]
 m202_dyn_flap_front [{shapeType=Box, extentX=26.4284, extentY=9.1644, extentZ=28.3568, center=<-0.217 -1.804 -12.8082>}]
}
constraints [
 {
  bone1Name=m202_dyn_flap_front
  bone2Name=handle
  swing1LimitDegrees=0
  swing1Motion=Locked
  swing2LimitDegrees=0
  swing2Motion=Locked
  twistLimitDegrees=45
 }
 {
  bone1Name=m202_dyn_flap_back
  bone2Name=handle
  swing1LimitDegrees=0
  swing1Motion=Locked
  swing2LimitDegrees=0
  swing2Motion=Locked
  twistLimitDegrees=45
 }
]
