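# Physics asset for the Pumpkin Launcher rocket launcher skeletal mesh, giving the weapon simplified collision.
# One large offset box on the Handle bone provides the main volume, while a tiny box sits on b_AR_Root.
# A single constraint joins the two bodies with swing and twist set free, so the listed angular limits do not restrict rotation.
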PhysicsAsset /Game/Weapons/FORT_RocketLaunchers/Mesh/SK_PumpkinLauncher_Physics
bodies {
 Handle [{shapeType=Box, extentX=135.9657, extentY=25.0899, extentZ=38.5961, center=<19.9346 2.4661 11.0293>}]
 b_AR_Root [{shapeType=Box, extentX=1.01, extentY=1.01, extentZ=1.01}]
}
constraints [{
  bone1Name=Handle
  bone2Name=b_AR_Root
  swing1LimitDegrees=45
  swing1Motion=Free
  swing2LimitDegrees=45
  swing2Motion=Free
  twistLimitDegrees=45
  twistMotion=Free
 }]
