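# Physics asset for the RPG-7 rocket launcher skeletal mesh, giving the weapon simple collision shapes.
# Two bodies each use a single box: a large one on the Handle bone offset to cover the launcher, and a smaller one centered on the Rocket bone.
# There are no constraints, so the two bodies are not jointed to each other.
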PhysicsAsset /Game/Weapons/FORT_RocketLaunchers/Mesh/SK_RPG7_Physics2
bodies {
 Handle [{shapeType=Box, extentX=102.2435, extentY=17.5339, extentZ=39.1088, center=<-5.6676 2.0259 11.7294>}]
 Rocket [{shapeType=Box, extentX=62.2845, extentY=14.0175, extentZ=13.8045}]
}
constraints []
